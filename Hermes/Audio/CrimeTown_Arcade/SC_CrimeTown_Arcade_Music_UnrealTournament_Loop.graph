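# Sound cue for looping Unreal Tournament music in the CrimeTown arcade.
# A single looping wave player feeds the output directly, using interior music attenuation and the indoor ambient sound class with boosted volume.
# It overrides concurrency to allow two instances, stopping the quietest, and has a send to the default Fortnite reverb submix.
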SoundCue /Hermes/Audio/CrimeTown_Arcade/SC_CrimeTown_Arcade_Music_UnrealTournament_Loop
root=WavePlayer_0
nodes [{id=WavePlayer_0, properties {bLooping=true, soundWaveAssetPtr=&/MX_S34_Arcade_UnrealTournament_LP_01}}]
properties {
 attenuationSettings=&/ATT_CrimeTown_Arcade_Music_Interior
 bOverrideConcurrency=true
 maxDistance=1000
 soundClassObject=&Fort_Ambient_Indoors_Athena
 volumeMultiplier=1.6
 concurrencyOverrides {maxCount=2, resolutionRule=StopQuietest}
 soundSubmixSends [{minSendLevel=0, soundSubmix=&Fortnite_Reverb_Default}]
}
